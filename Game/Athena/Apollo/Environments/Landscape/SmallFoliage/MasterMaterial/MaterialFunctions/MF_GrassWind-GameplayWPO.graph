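# Material function for Apollo small-foliage grass that outputs wind sway, wind normals and gameplay-driven world position offset.
# A panned world-space swirl-noise normal map drives sway, with speed and power blended toward storm values by MF_StormMask.
# FinalWPO blends the AfterHeightModification input with player bending, pushes grass down and away near the camera, and sinks fading instances along their normals.
# The sliding blend and alternate wind branch are unconnected.
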MaterialFunction /Game/Athena/Apollo/Environments/Landscape/SmallFoliage/MasterMaterial/MaterialFunctions/MF_GrassWind-GameplayWPO
graph {nodes [
  CollectionParameter() {collection=/Game/Athena/Environments/Landscape/MPC/MPC_Landscape, parameterName=PlayerBend_Strength}
  CollectionParameter() {collection=&MPC_Landscape, parameterName=PlayerBend_Radius}
  Constant() {desc=Default, r=0.75}
  Constant() {desc=Default, r=0.4}
  Constant() {desc=Default, r=0.4}
  Constant() {desc=Default, r=1}
  Constant() {r=-50}
  Constant() {desc=Default, r=0.08}
  n1 = FunctionInput(Preview=Constant() {r=1}) {bUsePreviewValueAsDefault=true, inputName=WindFalloffMask, inputType=FunctionInput_Scalar, sortPriority=1}
  n2 = CollectionParameter() {collection=&MPC_Landscape, parameterName=Ambient_WindMotion_Amount}
  n3 = Multiply(A=@n1, B=@n2)
  n4 = ComponentMask(None=WorldPosition().XYZ) {g=true, r=true}
  n5 = Constant() {r=1200}
  n6 = MaterialFunctionCall() {materialFunction=/Game/Athena/Environments/Landscape/MaterialFunctions/MF_StormMask}
  n7 = LinearInterpolate(
   A=CollectionParameter() {collection=&MPC_Landscape, parameterName=Grass_WindSpeed}
   B=CollectionParameter() {collection=&MPC_Landscape, parameterName=Grass_WindSpeed_Storm}
   Alpha=@n6.Result
  )
  n8 = Time()
  n9 = TextureSample(UVs=Panner(Coordinate=Divide(A=@n4, B=@n5), Time=Multiply(A=@n7, B=@n8)) {speedY=1}) {samplerSource=SSM_Wrap_WorldGroupSettings, samplerType=SAMPLERTYPE_Normal, texture=/Game/Global/Textures/Default/Noise/T_SwirlNoise_N}
  n10 = LinearInterpolate(
   A=CollectionParameter() {collection=&MPC_Landscape, parameterName=Grass_WindPower}
   B=CollectionParameter() {collection=&MPC_Landscape, parameterName=Grass_WindPower_Storm}
   Alpha=@n6.Result
  )
  n11 = Multiply(A=@n9.G, B=@n10)
  FunctionOutput(None=Multiply(A=Multiply(A=@n3, B=@n11), B=Constant3Vector() {constant=<rgba 0 300 0 0>})) {outputName=Wind, sortPriority=1}
  FunctionOutput(None=@n9.RGB) {outputName=WindNormals, sortPriority=2}
  n12 = CollectionParameter() {collection=&FortniteMaterialParameters, parameterName=GrassPushDownRadius}
  n13 = FunctionInput(Preview=@n12) {bUsePreviewValueAsDefault=true, description="Default = 200", inputName=SinkIntoGroundDistance, inputType=FunctionInput_Scalar}
  n14 = Subtract(A=WorldPosition().XYZ, B=CameraPositionWS())
  n15 = Constant3Vector() {constant=<rgba 1 1 0 0>}
  n16 = Multiply(
   A=SphereMask(A=CameraPositionWS(), B=WorldPosition().XYZ, Radius=@n13, Hardness=Constant())
   B=Add(A=Multiply(A=@n14, B=@n15), B=Constant3Vector() {constant=<rgba 0 0 -200 0>})
  )
  n17 = CollectionParameter() {collection=&FortniteMaterialParameters, parameterName=PlayerPosition}
  n18 = MaterialFunctionCall!(PlayerPosition=@n17, PlayerRadiusMax=Constant() {r=200}, BendStrength=Constant() {r=-50}) {materialFunction=/Game/Athena/Environments/Landscape/MaterialFunctions/MF_AthenaFoliagePlayerInteract}
  n19 = LinearInterpolate(
   A=Multiply(A=FunctionInput() {inputName=AfterHeightModification, sortPriority=2}, B=PerInstanceFadeAmount())
   B=Multiply(A=@n18.Rotation, B=Constant() {r=0.3})
   Alpha=@n18.PlayerSphereMask
  )
  n20 = LinearInterpolate(
   A=ScalarParameter() {defaultValue=-50, group=WPO-DistanceBased, parameterName=SinkIntoGroundAmount}
   B=Constant()
   Alpha=PerInstanceFadeAmount()
  )
  n21 = VertexNormalWS()
  FunctionOutput(None=Add(A=Add(A=@n19, B=@n16) {constB=0}, B=Multiply(A=@n20, B=@n21))) {outputName=FinalWPO, sortPriority=3}
  n22 = CollectionParameter() {collection=&FortniteMaterialParameters, parameterName=IsPlayerSliding}
  LinearInterpolate(A=@n16, B=Constant(), Alpha=@n22)
  n23 = Multiply(A=@n1, B=@n2)
  n24 = Constant() {r=300}
  n25 = Multiply(A=@n9.RGB, B=Constant3Vector() {constant=<rgba 0 1 0 0>})
  n26 = AppendVector(A=ComponentMask(None=Multiply(A=@n25, B=@n10)) {g=true, r=true}, B=Constant() {r=1})
  Multiply(A=Multiply(A=Multiply(A=@n23, B=@n24), B=Normalize(VectorInput=@n26)), B=Constant3Vector() {constant=<rgba 1 1 0 0>})
 ]}
