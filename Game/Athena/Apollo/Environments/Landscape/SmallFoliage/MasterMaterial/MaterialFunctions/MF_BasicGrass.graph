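# Material function for Apollo landscape grass meshes, switched in by Use-MF_BasicGrass or passing input attributes through.
# Base attributes come from the landscape RVT or common grass layer functions, then gain per-instance tint, bright tips, vertex-red flowers and distance-based specular.
# World position offset combines storm-aware panning-noise wind, player and camera push-down, distance-fade sinking and physics fields.
# Forward and mobile paths use simpler shading; several debug and experimental nodes remain unconnected.
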MaterialFunction /Game/Athena/Apollo/Environments/Landscape/SmallFoliage/MasterMaterial/MaterialFunctions/MF_BasicGrass
graph {nodes [
  n1 = MaterialFunctionCall(Float3=NamedRerouteUsage()) {materialFunction=&BreakOutFloat3Components}
  Abs(None=Subtract(A=Abs(None=@n1.R), B=Abs(None=@n1.G)))
  n2 = ComponentMask(None=NamedRerouteUsage()) {g=true}
  Abs(None=@n2)
  Add(A=Saturate(None=Subtract(A=@n1.R, B=@n1.R)), B=Saturate(None=Subtract(A=@n1.G, B=@n1.G)))
  n3 = NamedRerouteUsage()
  n4 = Multiply(A=Time(), B=ScalarParameter() {defaultValue=0.3, parameterName=r4ythji0})
  n5 = Constant2Vector() {g=1}
  n6 = TextureCoordinate() {coordinateIndex=1}
  n7 = ScalarParameter() {defaultValue=1, parameterName=asdgt54yukb6722}
  n8 = Add(A=PerInstanceRandom(), B=Multiply(A=Multiply(A=@n6, B=@n7), B=ScalarParameter() {defaultValue=1, parameterName=lolwut}))
  n9 = Divide(
   A=ComponentMask(None=WorldPosition() {worldPositionShaderOffset=WPT_ExcludeAllShaderOffsets}.XYZ) {g=true, r=true}
   B=ScalarParameter() {defaultValue=200, group=WPO, parameterName=WindWaveScale33}
  )
  n10 = TextureSample(UVs=Add(A=Multiply(A=@n4, B=@n5), B=Add(A=@n8, B=@n9))) {
   desc="Replace this with:\r\nT_TorusFlowmap"
   samplerSource=SSM_Wrap_WorldGroupSettings
   texture=/Game/Athena/Apollo/Environments/Landscape/SmallFoliage/Assets/BasicGrass/T_ApolloGrass01
  }
  n11 = ComponentMask(None=Multiply(A=Subtract(A=@n10.RGB) {constB=0.5}, B=VectorParameter() {parameterName=hgre89ooo, defaultValue=<rgba 0.1 0 0 1>}.RGB)) {g=true, r=true}
  n12 = ConstantBiasScale(None=@n1.G)
  n13 = ScalarParameter() {defaultValue=2, parameterName=yrtyurth321}
  Add(A=TextureCoordinate(), B=LinearInterpolate(A=Multiply(A=@n3, B=@n11), Alpha=Power(Base=@n12, Exp=@n13)) {constB=0})
  AppendVector(A=Constant() {desc=empty}, B=AppendVector(A=Constant() {desc=wind, r=300}, B=Constant() {desc=vertical, r=1}))
  Constant3Vector() {constant=<rgba 0.6 0.7258 0.8 1>}
  Constant() {r=2}
  Constant() {r=0.5}
  Constant() {r=0.9}
  Constant() {r=2}
  Constant() {r=1.5}
  Constant() {r=0.2}
  Constant() {r=5000}
  Constant() {desc=Tip brightness, r=1.3}
  Constant() {r=0.6}
  Constant() {r=0.2}
  Constant() {r=-2}
  Constant() {r=0.3}
  Constant() {r=2}
  Constant() {r=4}
  Constant() {r=-0.5}
  Constant() {r=-50}
  Constant() {desc=Default, r=0.08}
  Constant() {desc=Default, r=0.75}
  Constant() {desc=Default, r=0.4}
  Constant() {desc=Default, r=0.4}
  Constant() {desc=Default, r=1}
  Constant() {r=3000}
  n14 = Constant() {r=0.1}
  n15 = SetMaterialAttributes!(
   MaterialAttributes
   BaseColor=NamedRerouteUsage()
   Specular=NamedRerouteUsage()
   Roughness=NamedRerouteUsage()
   OpacityMask=NamedRerouteUsage()
   Normal=NamedRerouteUsage()
   WorldPositionOffset=NamedRerouteUsage()
   Opacity=@n14
  ) {attributeSetTypes [
    #1
    #2
    #3
    #4
    #5
    #6
    #7
   ]}
  n16 = SetMaterialAttributes(
   BaseColor=NamedRerouteUsage()
   Specular=NamedRerouteUsage()
   Roughness=Constant() {r=0.9}
   OpacityMask=NamedRerouteUsage()
   Normal=VertexNormalWS()
   WorldPositionOffset=NamedRerouteUsage()
   Opacity=@n14
  ) {attributeSetTypes [
    #1
    #2
    #3
    #4
    #5
    #6
    #7
   ]}
  FunctionOutput(None=StaticSwitchParameter(
    True=ShadingPathSwitch(Default=@n15, Deferred=@n15, Forward=@n16, Mobile=@n16)
    False=FunctionInput() {inputType=FunctionInput_MaterialAttributes}
   ) {group=ChooseMaterial, parameterName=Use-MF_BasicGrass})
  n17 = VertexNormalWS()
  LinearInterpolate(A=@n17, B=NamedRerouteUsage(), Alpha=ScalarParameter() {defaultValue=1, parameterName=ghrftji678544})
  MaterialFunctionCall(
   1=Constant3Vector() {constant=<rgba 0.2243 0 1 1>}
   2=Constant3Vector() {constant=<rgba 0 1 0 0>}
   3=Constant3Vector() {constant=<rgba 0 0 1 0>}
   4=Constant3Vector() {constant=<rgba 1 0.5 0 1>}
   A=PerInstanceRandom()
  ) {materialFunction=&Lerp_Multiple_Float3}
  n18 = SphereMask(
   A=CameraPositionWS()
   B=WorldPosition().XYZ
   Radius=ScalarParameter() {defaultValue=3000, parameterName=CM-GrassTipMask-DistAdd}
   Hardness=Constant()
  )
  n19 = PerInstanceRandom()
  Max(A=@n18, B=@n19)
  n20 = VertexColor() {desc=Remove tall weeds}
  n21 = OneMinus(None=@n20.G)
  n22 = NamedRerouteUsage()
  n23 = Constant() {r=2}
  n24 = Saturate(None=Add(A=Multiply(A=@n22, B=@n23), B=Constant() {r=-1.25}))
  n25 = Constant() {r=0.6}
  n26 = Saturate(None=Add(A=NamedRerouteUsage(), B=Multiply(A=@n24, B=@n25)))
  n27 = LinearInterpolate(A=Constant() {r=-5}, B=Constant() {r=100}, Alpha=@n26)
  n28 = Constant() {r=1}
  n29 = ShadingPathSwitch(Default=@n27, Deferred=@n27, Forward=@n28, Mobile=@n28)
  Multiply(A=@n21, B=@n29)
  Multiply()
  n30 = Constant() {r=1}
  n31 = Constant() {r=-0.3}
  n32 = ComponentMask(None=TextureCoordinate() {coordinateIndex=1}) {g=true}
  NamedRerouteDeclaration(None=Saturate(None=LinearInterpolate(A=@n30, B=@n31, Alpha=@n32) {constA=1.2, constB=-1})) {name=UVHeightMask1, nodeColor=<rgba 1 0 1 1>}
  NamedRerouteDeclaration(None=@n32) {name=UV1-G, nodeColor=<rgba 1 0 1 1>}
  n33 = WorldPosition() {desc=This should stop the grass from changing color as the wind blows it around, worldPositionShaderOffset=WPT_ExcludeAllShaderOffsets}
  n34 = RuntimeVirtualTextureSampleParameter!(UVs, World Position=@n33.XYZ, Mip Bias=Constant() {r=3}) {
   desc="The spec from the RVT has strange mipping\r\nissues when viewing glancing angles of the\r\ngrass mesh!"
   materialType=BaseColor_Normal_Specular_YCoCg
   mipValueMode=RVTMVM_MipBias
   parameterName=RVT
   virtualTexture=/Game/Athena/Apollo/Maps/Landscape/VT/RVT_Apollo_Landscape
  }
  n35 = SphereMask(A=CameraPositionWS(), B=WorldPosition().XYZ, Radius=Constant() {r=2500}, Hardness=Constant())
  n36 = Constant() {r=0.1}
  n37 = Constant() {r=0.4}
  n38 = OneMinus(None=NamedRerouteUsage())
  n39 = Constant() {r=3.75}
  n40 = Saturate(None=Add(A=Multiply(A=@n38, B=@n39) {constB=3.75}, B=Constant() {r=-2}) {constB=-2})
  n41 = NamedRerouteUsage()
  n42 = Multiply(A=Add(A=Constant() {r=0.8}, B=OneMinus(None=@n41)), B=Max(A=Constant() {r=0.85}, B=NamedRerouteUsage()))
  n43 = NamedRerouteUsage()
  n44 = SetMaterialAttributes!(
   MaterialAttributes
   BaseColor=@n34.BaseColor
   Specular=Multiply(A=OneMinus(None=@n35), B=LinearInterpolate(A=@n36, B=@n37, Alpha=@n40))
   Roughness=Saturate(None=Max(A=@n42, B=@n43))
   Normal=VertexNormalWS()
  ) {attributeSetTypes [
    #1
    #2
    #3
    #5
   ]}
  n45 = SetMaterialAttributes(BaseColor=@n34.BaseColor, Specular=@n34.BaseColor, Roughness=@n34.BaseColor, Normal=@n34.BaseColor) {attributeSetTypes [
    #1
    #2
    #3
    #5
   ]}
  n46 = SetMaterialAttributes!(
   MaterialAttributes=MaterialFunctionCall!(CombinedLandscapeLayersMask, BypassUseRVT, BypassBiomeColors) {materialFunction=/Game/Athena/Apollo/Environments/Landscape/Materials/Biome/CommonMaterials/Mobile/MF_Mobile_Common_Grass}.ForGrassMeshes
   Normal=VertexNormalWS()
  ) {attributeSetTypes [#5]}
  n47 = StaticSwitchParameter(
   True=@n46
   False=MaterialFunctionCall(NoRVT=@n46, RVT=QualitySwitch(Default=@n44, Epic=@n45)) {materialFunction=&FN_RVT_Switch_Material}.Result
  ) {
   desc="Use True for spawning on static mesh terrain\r\nor if using the RVT is not desired"
   group=MF_BasicGrass
   parameterName=IsStaticMeshFoliageSpawner
  }
  n48 = GetMaterialAttributes(None=@n47) {attributeGetTypes [
    #1
    #5
    #3
    #2
   ]}
  NamedRerouteDeclaration(None=@n48.Specular) {name=IncomingSpec, nodeColor=<rgba 1 0 1 1>}
  n49 = PerInstanceRandom()
  n50 = Saturate(None=Add(A=Multiply(B=@n49) {constA=5}) {constB=-0.5})
  n51 = LinearInterpolate(A=Constant3Vector() {constant=<rgba 0.3628 0.1375 0.6 1>}, B=Constant3Vector() {constant=<rgba 0.4 0.3133 0.008 1>}, Alpha=@n50)
  n52 = Constant3Vector() {constant=<rgba 0.0906 0.2413 0.9062 1>}
  n53 = Constant3Vector() {constant=<rgba 0.0771 0.6188 0.7708 1>}
  n54 = MaterialFunctionCall!(UseLandscapeCoords) {materialFunction=/Game/Athena/Apollo/Environments/Landscape/Materials/Biome/CommonMaterials/MF_ApolloTerrainMasks}
  n55 = StaticSwitchParameter(
   True=LinearInterpolate(
    A=@n51
    B=LinearInterpolate(A=@n52, B=@n53, Alpha=@n50)
    Alpha=ChannelMaskParameter(None=@n54.BiomeMaskRGBA) {maskChannel=Alpha, parameterName=BiomeMaskChannel, defaultValue=<rgba 0 0 0 1>}.R
   )
   False=@n51
  ) {group=MF_BasicGrass, parameterName=UseBiomeMaskFlowerColor}
  n56 = NamedRerouteUsage()
  n57 = Multiply(A=NamedRerouteUsage(), B=Constant() {desc=Tip brightness, r=0.9})
  n58 = Multiply(A=NamedRerouteUsage(), B=SphereMask(A=CameraPositionWS(), B=WorldPosition().XYZ, Radius=Constant() {r=2500}, Hardness=Constant()))
  n59 = LinearInterpolate(A=@n57, B=Multiply(A=@n57, B=Constant() {desc=Tip brightness, r=2}), Alpha=@n58)
  n60 = Constant() {desc=Tip brightness, r=0.95}
  n61 = Constant() {desc=Tip brightness, r=1.2}
  n62 = PerInstanceRandom()
  n63 = Saturate(None=Divide(A=Saturate(None=@n59), B=LinearInterpolate(A=@n60, B=@n61, Alpha=@n62)))
  n64 = NamedRerouteUsage()
  n65 = ScalarParameter() {defaultValue=0.9, group=MF_BasicGrass, parameterName=BrightBase-Default}
  n66 = ScalarParameter() {defaultValue=2, desc=Tip brightness, group=MF_BasicGrass, parameterName=BrightTip-Default}
  n67 = LinearInterpolate(A=Constant() {desc=Tip brightness, r=0.5}, Alpha=PerInstanceRandom()) {constA=0.5}
  n68 = NamedRerouteUsage()
  n69 = NamedRerouteUsage()
  n70 = LinearInterpolate(
   A=QualitySwitch(
    Default=@n65
    Low=@n65
    High=@n65
    Medium=@n65
    Epic=ScalarParameter() {defaultValue=0.6, group=MF_BasicGrass, parameterName=BrightBase-Epic}
   )
   B=QualitySwitch(
    Default=@n66
    Low=@n66
    High=@n66
    Medium=@n66
    Epic=ScalarParameter() {defaultValue=1.3, desc=Tip brightness, group=MF_BasicGrass, parameterName=BrightTip-Epic}
   )
   Alpha=Multiply(A=Multiply(A=@n67, B=@n68), B=OneMinus(None=@n69))
  )
  n71 = SphereMask(A=CameraPositionWS(), B=WorldPosition().XYZ, Radius=Constant() {r=4000}, Hardness=Constant())
  n72 = VertexNormalWS()
  n73 = CameraVectorWS()
  n74 = OneMinus(None=Clamp(None=DotProduct(A=@n72, B=@n73)))
  n75 = QualitySwitch(Default=@n71, Low=@n71, High=@n71, Medium=@n71, Epic=Multiply(A=@n71, B=@n74))
  n76 = Multiply(A=@n74, B=LinearInterpolate(A=ScalarParameter() {defaultValue=0.5, parameterName=DistanceBrightness}, Alpha=@n71))
  n77 = LinearInterpolate(
   A=LinearInterpolate(A=@n64, B=Multiply(A=@n64, B=@n70), Alpha=StaticSwitchParameter(True=@n76, False=@n75) {parameterName=BrightTipsOverrideDistanceClamp})
   B=LinearInterpolate(
    A=VectorParameter() {group=MF_BasicGrass, parameterName=RandomPatchColor1, defaultValue=<rgba 0.15 0.1163 0.015 1>}.RGB
    B=VectorParameter() {group=MF_BasicGrass, parameterName=RandomPatchColor2, defaultValue=<rgba 0.13 0.0877 0.039 1>}.RGB
    Alpha=PerInstanceRandom()
   )
   Alpha=Multiply(A=NamedRerouteUsage(), B=NamedRerouteUsage())
  )
  n78 = LinearInterpolate(A=@n77, B=Multiply(A=@n77, B=Constant() {r=1.25}), Alpha=Multiply(A=NamedRerouteUsage(), B=VertexColor().G))
  NamedRerouteDeclaration(None=LinearInterpolate(A=ShadingPathSwitch(Default=@n78, Forward=@n63, Mobile=@n63), B=Multiply(A=@n55, B=@n56), Alpha=VertexColor().R)) {name=FinalBasecolor, nodeColor=<rgba 1 0 1 1>}
  n79 = TextureSampleParameter2D() {parameterName=MaskTexture, texture=&T_ApolloGrass01}
  NamedRerouteDeclaration(None=Multiply(A=@n79.A, B=ScalarParameter() {defaultValue=4, desc=Multiply alpha channel mask, group=MF_BasicGrass, parameterName=OpacityMask-Push}) {constB=4}) {name=FinalOpacityMask, nodeColor=<rgba 1 0 1 1>}
  n80 = Constant() {desc=Red channel}
  n81 = Saturate(None=LinearInterpolate(A=Constant() {r=1}, B=Constant() {r=-3}, Alpha=NamedRerouteUsage()) {constA=1.2, constB=-1})
  n82 = NamedRerouteUsage()
  n83 = VertexColor() {desc=Remove tall weeds}
  n84 = NamedRerouteUsage()
  n85 = Constant() {r=2}
  n86 = Saturate(None=Add(A=Multiply(A=@n84, B=@n85), B=Constant() {r=-1.25}))
  n87 = Constant() {r=0.6}
  n88 = Saturate(None=Add(A=NamedRerouteUsage(), B=Multiply(A=@n86, B=@n87)))
  n89 = Constant() {r=0.5}
  n90 = ShadingPathSwitch(Default=@n88, Forward=@n89, Mobile=@n89)
  n91 = CollectionParameter() {collection=/Game/Athena/Environments/Landscape/MPC/MPC_Landscape, parameterName=Ambient_WindMotion_Amount}
  n92 = ComponentMask(None=WorldPosition().XYZ) {g=true, r=true}
  n93 = Constant() {r=1200}
  n94 = MaterialFunctionCall() {materialFunction=/Game/Athena/Environments/Landscape/MaterialFunctions/MF_StormMask}
  n95 = LinearInterpolate(
   A=CollectionParameter() {collection=&MPC_Landscape, parameterName=Grass_WindSpeed}
   B=CollectionParameter() {collection=&MPC_Landscape, parameterName=Grass_WindSpeed_Storm}
   Alpha=@n94.Result
  )
  n96 = Time()
  n97 = TextureSample(UVs=Panner(Coordinate=Divide(A=@n92, B=@n93), Time=Multiply(A=@n95, B=@n96)) {speedY=1}) {samplerSource=SSM_Wrap_WorldGroupSettings, samplerType=SAMPLERTYPE_Normal, texture=/Game/Global/Textures/Default/Noise/T_SwirlNoise_N}
  n98 = LinearInterpolate(
   A=CollectionParameter() {collection=&MPC_Landscape, parameterName=Grass_WindPower}
   B=CollectionParameter() {collection=&MPC_Landscape, parameterName=Grass_WindPower_Storm}
   Alpha=@n94.Result
  )
  n99 = Multiply(A=Multiply(A=Multiply(A=@n90, B=@n91), B=Multiply(A=@n97.G, B=@n98)), B=Constant() {r=300})
  n100 = LinearInterpolate(A=Constant() {r=-5}, B=Constant() {r=100}, Alpha=@n88)
  n101 = Constant() {r=1}
  n102 = Multiply(A=OneMinus(None=@n83.G), B=ShadingPathSwitch(Default=@n100, Deferred=@n100, Forward=@n101, Mobile=@n101)) {desc=Blue channel}
  n103 = Multiply(A=Multiply(A=LinearInterpolate(A=@n81, B=@n82, Alpha=@n83.G), B=AppendVector(A=@n99, B=@n102)), B=PerInstanceFadeAmount()) {desc="Fade out wind && height offset in distance"}
  n104 = Constant3Vector() {constant=<rgba -22.5 -22.5 -4.5 1>}
  n105 = CollectionParameter() {collection=&FortniteMaterialParameters, parameterName=PlayerPosition}
  n106 = ComponentMask(None=@n105) {b=true, g=true, r=true}
  n107 = Normalize(VectorInput=Subtract(A=@n106, B=WorldPosition().XYZ))
  n108 = SphereMask(A=@n106, B=WorldPosition().XYZ, Radius=Constant() {r=200}, Hardness=Constant()) {attenuationRadius=64, hardnessPercent=20}
  n109 = MaterialFunctionCall(Color=LinearInterpolate(A=AppendVector(A=@n80, B=@n103), B=Multiply(A=@n104, B=@n107), Alpha=Saturate(None=@n108) {desc=PlayerSphereMask})) {materialFunction=&SplitComponents}
  n110 = CollectionParameter() {collection=&FortniteMaterialParameters, parameterName=GrassPushDownRadius}
  n111 = Multiply(A=SphereMask(A=CameraPositionWS(), B=WorldPosition().XYZ, Radius=@n110, Hardness=Constant()), B=Constant() {r=-200})
  n112 = Constant()
  n113 = CollectionParameter() {collection=&FortniteMaterialParameters, parameterName=IsPlayerSliding}
  n114 = Add(A=Add(A=@n109.B, B=LinearInterpolate(A=@n111, B=@n112, Alpha=@n113)), B=Constant() {r=-2.5}) {desc=move mesh down a bit}
  n115 = LinearInterpolate(
   A=ScalarParameter() {defaultValue=-50, group=WPO-DistanceBased, parameterName=SinkIntoGroundAmount}
   B=Constant()
   Alpha=PerInstanceFadeAmount()
  )
  n116 = VertexNormalWS()
  n117 = MaterialFunctionCall!(
   Previous WPO Value=Add(
    A=MaterialFunctionCall(X=@n109.R, Y=@n109.R, Z=@n114) {materialFunction=&MakeFloat3}.Result
    B=Multiply(A=@n115, B=@n116)
   )
   Force Magnitude Mask
   Force Magnitude Multiplier
   Project Field Force Vector to Normals
   Noise Size
   Noise Pan Speed
   Noise Falloff
   Stretch Noise Coordinates Z
  ) {desc=Sample global physics field add to WPO, materialFunction=/Game/Effects/Niagara/MasterMaterials/Functions/MF_FX_PhysicsFields_Grass}
  NamedRerouteDeclaration(None=@n117.Result) {name=FinalWPO, nodeColor=<rgba 1 0 1 1>}
  n118 = NamedRerouteUsage()
  n119 = Multiply(A=NamedRerouteUsage(), B=Constant() {r=1.5})
  n120 = Constant() {r=-0.3}
  n121 = Multiply(A=Saturate(None=Add(A=@n119, B=@n120)), B=@n118)
  n122 = SphereMask(A=CameraPositionWS(), B=WorldPosition().XYZ, Radius=Constant() {r=30000}, Hardness=Constant())
  n123 = Saturate(None=Multiply(A=ScalarParameter() {defaultValue=3, group=MF_BasicGrass, parameterName=Grs-SpecTextureDistMod}, B=@n122))
  n124 = ScalarParameter() {defaultValue=0.75, group=MF_BasicGrass, parameterName=Grs-SpecTextureMod12}
  n125 = Saturate(None=Multiply(A=@n123, B=ScalarParameter() {defaultValue=1.5, group=MF_BasicGrass, parameterName=Grs-SpecTextureMod1}))
  n126 = LinearInterpolate(
   A=ScalarParameter() {defaultValue=2, group=MF_BasicGrass, parameterName=3GrsSpecLowerNear2}
   B=ScalarParameter() {defaultValue=-2, group=MF_BasicGrass, parameterName=3GrsSpecLowerNear}
   Alpha=ConstantBiasScale(None=@n2)
  )
  n127 = Saturate(None=Add(A=@n126, B=ScalarParameter() {defaultValue=0.5, group=MF_BasicGrass, parameterName=Grs-SpecAddTowind}))
  n128 = Multiply(
   A=Max(A=Multiply(A=@n123, B=@n124), B=Multiply(A=@n125, B=@n127))
   B=Multiply(A=PerInstanceRandom(), B=NamedRerouteUsage()) {desc=New blade spec texture}
  )
  n129 = NamedRerouteUsage()
  n130 = Multiply(A=NamedRerouteUsage(), B=ScalarParameter() {defaultValue=2, group=MF_BasicGrass, parameterName=specwavemask-grad-mod})
  n131 = ScalarParameter() {defaultValue=-0.5, group=MF_BasicGrass, parameterName=specwavemask-grad-add}
  n132 = LinearInterpolate(A=@n126, B=@n129, Alpha=Saturate(None=Add(A=@n130, B=@n131)))
  n133 = Constant() {r=1}
  n134 = Multiply(A=LinearInterpolate(A=@n129, B=Saturate(None=@n132), Alpha=@n122) {desc="B=Close, A=Distant"}, B=Abs(None=Subtract(A=@n133, B=@n122)))
  NamedRerouteDeclaration(None=QualitySwitch(Default=LinearInterpolate(A=@n118, B=@n121, Alpha=@n122) {desc="B=Close, A=Distant"}, Epic=Max(A=@n128, B=@n134))) {name=FinalSpec, nodeColor=<rgba 1 0 1 1>}
  NamedRerouteDeclaration(None=@n97.RGB) {name=WindNormals, nodeColor=<rgba 1 0 1 1>}
  n135 = Constant() {desc=Tip brightness, r=0.1}
  n136 = Constant() {desc=Tip brightness, r=0.6}
  NamedRerouteDeclaration(None=Saturate(None=LinearInterpolate(A=@n135, B=@n136, Alpha=@n58))) {name=MobileSpec, nodeColor=<rgba 1 0 1 1>}
  n137 = MaterialFunctionCall!(CombinedLandscapeLayersMask, BypassUseRVT, BypassBiomeColors) {materialFunction=/Game/Athena/Apollo/Environments/Landscape/Materials/Biome/CommonMaterials/MF_Common_Grass}
  NamedRerouteDeclaration(None=@n137.GrassColorMask3-NoMod) {name=GrassColorMask3-NoMod, nodeColor=<rgba 1 0 1 1>}
  NamedRerouteDeclaration(None=@n79.R) {name=GrassSpec-FlowerDiffuse, nodeColor=<rgba 1 0 1 1>}
  NamedRerouteDeclaration(None=OneMinus(None=@n32)) {name=UV1-G-Invert, nodeColor=<rgba 1 0 1 1>}
  n138 = Constant() {r=-2}
  n139 = Constant() {r=5}
  NamedRerouteDeclaration(None=Saturate(None=LinearInterpolate(A=@n138, B=@n139, Alpha=@n137.GrassColorMask4))) {name=GrassColorMask4-Mod, nodeColor=<rgba 1 0 1 1>}
  NamedRerouteDeclaration(None=@n48.BaseColor) {name=IncomingDiffuse, nodeColor=<rgba 1 0 1 1>}
  NamedRerouteDeclaration(None=@n48.Normal) {name=IncomingNormal, nodeColor=<rgba 1 0 1 1>}
  NamedRerouteDeclaration(None=@n137.GrassColorMask1) {name=GrassColorMask1, nodeColor=<rgba 1 0 1 1>}
  NamedRerouteDeclaration(None=@n137.GrassColorMask2) {name=GrassColorMask2, nodeColor=<rgba 1 0 1 1>}
  NamedRerouteDeclaration(None=@n137.GrassColorMask4) {name=GrassColorMask4, nodeColor=<rgba 1 0 1 1>}
  NamedRerouteDeclaration(None=@n48.Roughness) {name=IncomingRoughness, nodeColor=<rgba 1 0 1 1>}
  NamedRerouteUsage()
  n140 = Constant() {r=0.5}
  n141 = Saturate(None=LinearInterpolate(A=Constant() {r=1}, B=Constant() {r=-3}, Alpha=NamedRerouteUsage()) {constA=1.2, constB=-1})
  n142 = NamedRerouteUsage()
  n143 = Multiply(A=@n21, B=Multiply(A=Constant3Vector() {constant=<rgba 0 0 1 0>}, B=@n29))
  n144 = MaterialFunctionCall!(
   SinkIntoGroundDistance
   WindFalloffMask=ShadingPathSwitch(Default=@n26, Deferred=@n26, Forward=@n140, Mobile=@n140)
   AfterHeightModification=Multiply(A=LinearInterpolate(A=@n141, B=@n142, Alpha=@n20.G), B=Add(A=@n143, B=@n144.Wind))
  ) {materialFunction=&/MF_GrassWind-GameplayWPO}
  Reroute(None=@n144.WindNormals)
  Reroute(None=Add(A=@n144.FinalWPO, B=Constant3Vector() {constant=<rgba 0 0 -2.5 1>}) {desc=move mesh down a bit})
  n145 = NamedRerouteUsage()
  n146 = Saturate(None=Divide(A=@n145, B=VectorParameter() {parameterName=CreativeTipColor, defaultValue=<rgba 0.3 0.3629 0.4 1>}.RGB))
  n147 = ScalarParameter() {defaultValue=1.5, parameterName=CM-GrassTipMaskMod-Far}
  n148 = ScalarParameter() {defaultValue=1, parameterName=CM-GrassTipMask-Near}
  n149 = Multiply(A=VectorParameter() {parameterName=CreativeNormalMod1, defaultValue=<rgba 2 2 1 1>}.RGB, B=@n97.RGB)
  n150 = LinearInterpolate(
   A=Constant()
   B=Multiply(A=NamedRerouteUsage(), B=LinearInterpolate(A=@n147, B=@n148, Alpha=@n18))
   Alpha=MaterialFunctionCall(Normal Vector=Transform(None=@n149), Power=ScalarParameter() {defaultValue=2, parameterName=CM-GrassTipFresPow}) {materialFunction=&Fresnel_Function}.Result
  )
  n151 = PerInstanceRandom()
  n152 = Add(A=NamedRerouteUsage(), B=ScalarParameter() {parameterName=CM-GrassShadowAdd})
  n153 = Multiply(A=Multiply(A=Multiply(A=@n150, B=@n151), B=Saturate(None=@n152)), B=PerInstanceFadeAmount())
  n154 = Multiply(
   A=Power(Base=NamedRerouteUsage(), Exp=ScalarParameter() {defaultValue=3, parameterName=CM-GrassBaseMaskPow})
   B=ScalarParameter() {defaultValue=1, parameterName=CM-GrassTipMaskMod2}
  )
  n155 = ScalarParameter() {defaultValue=0.8, parameterName=CM-GrassShadowAdd2}
  n156 = PerInstanceFadeAmount()
  n157 = Saturate(None=Add(A=Add(A=@n154, B=@n155), B=OneMinus(None=@n156)))
  n158 = Constant() {r=1}
  Reroute(None=Multiply(A=LinearInterpolate(A=@n145, B=@n146, Alpha=@n153), B=LinearInterpolate(A=@n157, B=@n158, Alpha=@n19)))
  n159 = OneMinus(None=@n137.GrassColorMask3)
  n160 = Constant() {r=3}
  Saturate(None=LinearInterpolate(A=Constant() {r=0.1}, B=Constant() {r=0.4}, Alpha=Power(Base=@n159, Exp=@n160)))
  n161 = ScalarParameter() {defaultValue=1, parameterName=GRS-tw43tw}
  n162 = ScalarParameter() {defaultValue=0.5, parameterName=GRS-nrye5}
  n163 = MaterialFunctionCall(Normal Vector=@n17, Invert Fresnel=StaticBool(), Power=ScalarParameter() {defaultValue=5, parameterName=ge5y3e4tgeg}) {materialFunction=&Fresnel_Function}
  Saturate(None=Divide(A=@n78, B=LinearInterpolate(A=@n161, B=@n162, Alpha=@n163.Result)))
  ScalarParameter() {defaultValue=0.1, desc=Tip brightness, group=MF_BasicGrass, parameterName=Mobile-testspecvalue}
  ScalarParameter() {defaultValue=0.6, desc=Tip brightness, group=MF_BasicGrass, parameterName=Mobile-testspecvalue2}
  ScalarParameter() {defaultValue=1.5, group=MF_BasicGrass, parameterName=grs-specuv1mask-mod}
  ScalarParameter() {defaultValue=-0.3, group=MF_BasicGrass, parameterName=grs-specuv1mask-add}
  ScalarParameter() {defaultValue=-200, parameterName=debug-grass-wpo-pushdown-test33amt}
  ScalarParameter() {defaultValue=2500, parameterName=HHHH-testrad}
  ScalarParameter() {defaultValue=0.95, desc=Tip brightness, group=MF_BasicGrass, parameterName=BrightTip-Mobile-PIR1}
  ScalarParameter() {defaultValue=1.2, desc=Tip brightness, group=MF_BasicGrass, parameterName=BrightTip-Mobile-PIR2}
  ScalarParameter() {defaultValue=2, desc=Tip brightness, group=MF_BasicGrass, parameterName=BrightTip-Mobile}
  ScalarParameter() {defaultValue=0.9, desc=Tip brightness, group=MF_BasicGrass, parameterName=BrightTip-Mobile-boostdiff}
  ScalarParameter() {defaultValue=0.8, parameterName=DEBUG-RoughTipAdd}
  ScalarParameter() {defaultValue=0.85, parameterName=DEBUG-RoughMinA}
  ScalarParameter() {defaultValue=30000, group=MF_BasicGrass, parameterName=Terrain-GrsSpecHeightDist}
  ScalarParameter() {defaultValue=4000, group=MF_BasicGrass, parameterName=TipFadeGradDistance}
  n164 = Reroute()
  n165 = LinearInterpolate(
   A=VectorParameter() {group=Corrupted2, parameterName=Corrupted2-Color1, defaultValue=<rgba 0.1256 0.0217 0.1667 1>}.RGB
   B=VectorParameter() {group=Corrupted2, parameterName=Corrupted2-Color2, defaultValue=<rgba 0.0873 0.0408 0.1667 1>}.RGB
   Alpha=PerInstanceRandom()
  )
  n166 = LinearInterpolate(
   A=GetMaterialAttributes(None=@n164) {attributeGetTypes [#1]}.BaseColor
   B=@n165
   Alpha=ScalarParameter() {group=Corrupted2, parameterName=IsCorrupted2Grass}
  )
  ShadingPathSwitch(Default=@n164, Forward=@n166, Mobile=@n166)
  StaticBool() {value=true}
  StaticSwitchParameter(
   True=SetMaterialAttributes!(MaterialAttributes=@n164, BaseColor=@n165) {attributeSetTypes [#1]}
   False=@n164
  ) {parameterName=UseCorrupted2Grass}
  WorldPosition()
 ]}
