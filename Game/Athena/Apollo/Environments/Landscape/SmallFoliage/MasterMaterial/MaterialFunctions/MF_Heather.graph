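# Heather foliage material function for Apollo landscape small foliage, switchable against passthrough input attributes.
# Base color blends landscape RVT color into mask-driven heather colors by grass color mask and camera distance, with optional biome-varied flower tints.
# An RVT switch selects between this path and a simpler mobile set using vertex normals.
# Grass wind plus global physics fields drive world position offset, and a UV gradient sets opacity.
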MaterialFunction /Game/Athena/Apollo/Environments/Landscape/SmallFoliage/MasterMaterial/MaterialFunctions/MF_Heather
graph {nodes [
  Constant() {r=-0.1}
  n1 = SetMaterialAttributes!(
   MaterialAttributes
   BaseColor=NamedRerouteUsage()
   Specular=Constant() {r=0.25}
   Roughness=ScalarParameter() {defaultValue=0.8, group=MF_Heather, parameterName=Mobile-Roughness}
   OpacityMask=NamedRerouteUsage()
   Normal=VertexNormalWS()
   WorldPositionOffset=NamedRerouteUsage()
  ) {
   desc=Mobile
   attributeSetTypes [
    #1
    #2
    #3
    #4
    #5
    #6
   ]
  }
  n2 = NamedRerouteUsage()
  n3 = TextureSample() {
   samplerSource=SSM_Wrap_WorldGroupSettings
   samplerType=SAMPLERTYPE_Normal
   texture=/Game/Athena/Apollo/Environments/Landscape/SmallFoliage/Assets/Heather/Heather_N
  }
  n4 = MaterialFunctionCall(BaseNormal=@n2, AdditionalNormal=Transform(None=@n3.RGB)) {materialFunction=&BlendAngleCorrectedNormals}
  n5 = Multiply(A=ScalarParameter() {defaultValue=0.1, group=MF_Heather, parameterName=Heather-NormalBlend}, B=NamedRerouteUsage())
  n6 = ScalarParameter() {defaultValue=0.2, group=MF_Heather, parameterName=Heather-SpecularHigh}
  n7 = Multiply(
   A=ScalarParameter() {defaultValue=-0.2, group=MF_Heather, parameterName=Heather-SpecLow}
   B=LinearInterpolate(B=NamedRerouteUsage(), Alpha=Constant() {r=0.8}) {constA=1}
  )
  n8 = ComponentMask(None=TextureCoordinate() {coordinateIndex=1}) {g=true}
  n9 = SetMaterialAttributes(
   BaseColor=NamedRerouteUsage()
   Specular=Saturate(None=LinearInterpolate(A=@n6, B=@n7, Alpha=@n8))
   Roughness=Multiply(A=ScalarParameter() {defaultValue=1, group=MF_Heather, parameterName=Roughness}, B=NamedRerouteUsage())
   OpacityMask=NamedRerouteUsage()
   Normal=StaticSwitchParameter(True=LinearInterpolate(A=@n2, B=@n4.Result, Alpha=@n5), False=@n2) {defaultValue=true, group=MF_Heather, parameterName=UseCustomNormalTexture}
   WorldPositionOffset=NamedRerouteUsage()
  ) {attributeSetTypes [
    #1
    #2
    #3
    #4
    #5
    #6
   ]}
  n10 = MaterialFunctionCall(NoRVT=@n1, RVT=@n9) {materialFunction=&FN_RVT_Switch_Material}
  n11 = LinearInterpolate(
   A=ScalarParameter() {defaultValue=-1.1509, parameterName=SSSOpacity_Top}
   B=ScalarParameter() {defaultValue=10, parameterName=SSSOpacity_Bottom}
   Alpha=ComponentMask(None=TextureCoordinate()) {g=true}
  )
  FunctionOutput(None=StaticSwitchParameter(
    True=SetMaterialAttributes!(MaterialAttributes=@n10.Result, Opacity=@n11) {attributeSetTypes [#7]}
    False=FunctionInput() {inputType=FunctionInput_MaterialAttributes}
   ) {group=ChooseMaterial, parameterName=Use-MF_Heather})
  n12 = MaterialFunctionCall!(CombinedLandscapeLayersMask, BypassUseRVT, BypassBiomeColors) {materialFunction=/Game/Athena/Apollo/Environments/Landscape/Materials/Biome/CommonMaterials/MF_Common_Grass}
  n13 = Add(A=@n12.GrassColorMask1, B=Constant() {r=0.1})
  n14 = Constant() {r=0.5}
  n15 = LinearInterpolate(A=Constant() {r=1}, B=Constant() {r=-1}, Alpha=ComponentMask(None=TextureCoordinate() {coordinateIndex=1}) {g=true}) {constA=1.2, constB=-1}
  n16 = Constant() {r=1}
  n17 = Multiply(A=ShadingPathSwitch(Default=@n12.GrassColorMask1, Forward=@n16, Mobile=@n16), B=Constant3Vector() {constant=<rgba 0 0 1 0>})
  n18 = MaterialFunctionCall!(
   SinkIntoGroundDistance
   WindFalloffMask=ShadingPathSwitch(Default=Saturate(None=@n13), Forward=@n14, Mobile=@n14)
   AfterHeightModification=Multiply(A=Saturate(None=@n15), B=Add(A=@n17, B=@n18.Wind))
  ) {materialFunction=&/MF_GrassWind-GameplayWPO}
  n19 = MaterialFunctionCall!(
   Previous WPO Value=@n18.FinalWPO
   Force Magnitude Mask
   Force Magnitude Multiplier
   Project Field Force Vector to Normals
   Noise Size
   Noise Pan Speed
   Noise Falloff
   Stretch Noise Coordinates Z
  ) {desc=Sample global physics field add to WPO, materialFunction=/Game/Effects/Niagara/MasterMaterials/Functions/MF_FX_PhysicsFields_Grass}
  NamedRerouteDeclaration(None=@n19.Result) {name=WPO, nodeColor=<rgba 0.6172 0 1 1>}
  n20 = TextureSampleParameter2D() {
   group=MF_Heather
   parameterName=HeatherMaskTexture
   samplerSource=SSM_Wrap_WorldGroupSettings
   samplerType=SAMPLERTYPE_Masks
   texture=/Game/Athena/Apollo/Environments/Landscape/SmallFoliage/Assets/Heather/Heather_M
  }
  NamedRerouteDeclaration(None=Power(Base=@n20.R, Exp=ScalarParameter() {defaultValue=4, group=MF_Heather, parameterName=Heather-OPpow})) {name=OpacityMask, nodeColor=<rgba 0.6172 0 1 1>}
  n21 = Constant() {desc="This should be set to the same ## wherever the\r\nRVT is used in any material for the grass\r\nsystem for performance reasons!", r=3}
  n22 = RuntimeVirtualTextureSampleParameter!(UVs, World Position, Mip Bias=@n21) {
   materialType=BaseColor_Normal_Specular_YCoCg
   mipValueMode=RVTMVM_MipBias
   parameterName=RVT
   virtualTexture=/Game/Athena/Apollo/Maps/Landscape/VT/RVT_Apollo_Landscape
  }
  NamedRerouteDeclaration(None=@n22.Normal) {name=RVTNormal, nodeColor=<rgba 0.6172 0 1 1>}
  NamedRerouteDeclaration(None=@n22.Roughness) {name=RVTRough, nodeColor=<rgba 0.6172 0 1 1>}
  n23 = ScalarParameter() {defaultValue=-0.25, group=MF_Heather, parameterName=BlueChannelAdd-Mobile}
  n24 = NamedRerouteUsage()
  n25 = VectorParameter() {group=MF_Heather, parameterName=Heather-ColorA, defaultValue=<rgba 0.0529 0.135 0.0203 1>}
  n26 = VectorParameter() {group=MF_Heather, parameterName=HeatherGrasslands-ColorB, defaultValue=<rgba 0.4663 0.0825 0.55 1>}
  n27 = VectorParameter() {group=MF_Heather, parameterName=Heather-ColorB, defaultValue=<rgba 0.4663 0.0825 0.55 1>}
  n28 = MaterialFunctionCall!(UseLandscapeCoords) {materialFunction=/Game/Athena/Apollo/Environments/Landscape/Materials/Biome/CommonMaterials/MF_ApolloTerrainMasks}
  n29 = VectorParameter() {group=MF_Heather, parameterName=HeatherGrasslands-ColorB2, defaultValue=<rgba 0.4663 0.0825 0.55 1>}
  n30 = VectorParameter() {group=MF_Heather, parameterName=Heather-ColorB2, defaultValue=<rgba 0.4663 0.0825 0.55 1>}
  n31 = StaticSwitchParameter(
   True=LinearInterpolate(A=LinearInterpolate(A=@n26.RGB, B=@n27.RGB, Alpha=@n28.Desert), B=LinearInterpolate(A=@n29.RGB, B=@n30.RGB, Alpha=@n28.Desert), Alpha=@n20.B)
   False=@n27.RGB
  ) {group=MF_Heather, parameterName=UseMultiColor}
  n32 = Multiply(A=Saturate(None=Add(A=@n23, B=@n24)), B=LinearInterpolate(A=@n25.RGB, B=@n31, Alpha=NamedRerouteUsage()))
  NamedRerouteDeclaration(None=@n32) {name=MobileBasecolor, nodeColor=<rgba 0.6172 0 1 1>}
  NamedRerouteDeclaration(None=@n22.BaseColor) {name=RVTDiffuse, nodeColor=<rgba 0.6172 0 1 1>}
  NamedRerouteDeclaration(None=@n20.G) {name=Mask-Green, nodeColor=<rgba 0.6172 0 1 1>}
  NamedRerouteDeclaration(None=@n20.A) {name=Mask-Alpha, nodeColor=<rgba 0.6172 0 1 1>}
  n33 = VectorParameter() {group=MF_Heather, parameterName=Heather-UnderColor, defaultValue=<rgba 0.0529 0.135 0.0202 1>}
  n34 = ViewProperty() {property=MEVP_RuntimeVirtualTextureOutputLevel}
  n35 = ScalarParameter() {defaultValue=1.5, group=MF_Heather, parameterName=Heather-RVT_DistNear}
  n36 = LinearInterpolate(A=@n25.RGB, B=@n31, Alpha=Saturate(None=Multiply(A=@n34.InvProperty, B=@n35)))
  n37 = NamedRerouteUsage()
  n38 = ScalarParameter() {defaultValue=1, group=MF_Heather, parameterName=Heather-FlowerMaskMod}
  n39 = Saturate(None=Add(A=Multiply(A=@n12.GrassColorMask4, B=@n38), B=ScalarParameter() {group=MF_Heather, parameterName=Heather-FlowerMaskAdd}))
  n40 = NamedRerouteDeclaration(None=SphereMask(
    A=CameraPositionWS()
    B=WorldPosition().XYZ
    Radius=ScalarParameter() {defaultValue=6000, group=MF_Heather, parameterName=Heather-DistanceFade}
    Hardness=Constant()
   )) {name=CameraDistance, nodeColor=<rgba 0.6172 0 1 1>}
  n41 = ScalarParameter() {defaultValue=1, group=MF_Heather, parameterName=BlueChannelAdd}
  n42 = NamedRerouteUsage()
  n43 = Multiply(
   A=LinearInterpolate(A=NamedRerouteUsage(), B=LinearInterpolate(A=@n33.RGB, B=@n36, Alpha=@n37), Alpha=Multiply(A=@n39, B=@n40))
   B=Saturate(None=Add(A=@n41, B=@n42))
  )
  NamedRerouteDeclaration(None=@n43) {name=DefaultBasecolor, nodeColor=<rgba 0.6172 0 1 1>}
  Saturate(None=Divide(A=@n34.Property, B=ScalarParameter() {defaultValue=18, group=MF_Heather, parameterName=Heather-RVT_MipDistanceFarEffect}))
  n44 = ScalarParameter() {defaultValue=0.8, desc=Tip brightness, group=MF_Heather, parameterName=Mobile-PIR-A}
  n45 = ScalarParameter() {defaultValue=1.75, desc=Tip brightness, group=MF_Heather, parameterName=Mobile-PIR-B}
  n46 = PerInstanceRandom()
  Saturate(None=Divide(A=@n32, B=LinearInterpolate(A=@n44, B=@n45, Alpha=@n46)))
  n47 = ScalarParameter() {defaultValue=0.8, desc=Tip brightness, group=MF_Heather, parameterName=PIR-A}
  n48 = ScalarParameter() {defaultValue=1.75, desc=Tip brightness, group=MF_Heather, parameterName=PIR-B}
  n49 = PerInstanceRandom()
  Saturate(None=Divide(A=@n43, B=LinearInterpolate(A=@n47, B=@n48, Alpha=@n49)))
 ]}
