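# Optional material function layer for jungle small foliage, toggled by Use-MF_JunglePlants, otherwise passing input attributes through.
# MF_FoliageMaster2023Colors supplies the surface, with base color divided by per-instance random tints and specular, roughness and normals simplified on Forward/Mobile.
# WPO switches between MF_MediumPlantAnimation and landscape-masked, storm-aware noise wind, drops wind on Forward/Mobile, and offsets along vertex normals by instance fade.
# An older Nightmare-grass texture chain remains unconnected to the output.
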MaterialFunction /Game/Athena/Apollo/Environments/Landscape/SmallFoliage/MasterMaterial/MaterialFunctions/MF_JunglePlants
graph {nodes [
  Constant3Vector() {constant=<rgba 0.0784 0.1562 0.0448 0>}
  Constant() {desc=Default, r=0.4}
  Constant() {desc=Default, r=0.4}
  Constant() {desc=Default, r=1}
  Constant() {desc=Default, r=0.08}
  Constant() {desc=Default, r=0.75}
  n1 = GetMaterialAttributes(None=MaterialFunctionCall!(UseSpecMaskG_AsOpacityMask) {materialFunction=/Game/Global/Base4Layers_2022/MaterialFunctions/MF_FoliageMaster2023Colors}.Result) {attributeGetTypes [#1]}
  n2 = NamedRerouteUsage()
  n3 = SetMaterialAttributes!(MaterialAttributes=@n1.MaterialAttributes, BaseColor=Saturate(None=Divide(A=@n1.BaseColor, B=@n2))) {attributeSetTypes [#1]}
  n4 = GetMaterialAttributes(None=@n3) {attributeGetTypes [#2]}
  n5 = SetMaterialAttributes!(
   MaterialAttributes=@n3
   Specular=ScalarParameter() {defaultValue=0.4, group=MF_Leaves, parameterName=Mobile-Leaves-Specular}
   Roughness=ScalarParameter() {defaultValue=0.9, group=MF_Leaves, parameterName=Mobile-Leaves-Roughness}
   Normal=VertexNormalWS()
  ) {attributeSetTypes [#3, #4, #2]}
  n6 = ShadingPathSwitch(
   Default=SetMaterialAttributes!(MaterialAttributes=@n4.MaterialAttributes, Normal=Transform(None=@n4.Normal)) {attributeSetTypes [#2]}
   Forward=@n5
   Mobile=@n5
  )
  n7 = NamedRerouteUsage()
  FunctionOutput(None=StaticSwitchParameter(
    True=SetMaterialAttributes!(MaterialAttributes=@n6, WorldPositionOffset=@n7) {attributeSetTypes [#5]}
    False=FunctionInput() {inputType=FunctionInput_MaterialAttributes}
   ) {group=ChooseMaterial, parameterName=Use-MF_JunglePlants})
  n8 = LinearInterpolate(A=Constant() {r=0.4}, B=Constant(), Alpha=ComponentMask(None=TextureCoordinate() {coordinateIndex=1}) {g=true}) {constA=1.2, constB=-1}
  n9 = MaterialFunctionCall!(CombinedLandscapeLayersMask, BypassUseRVT, BypassBiomeColors) {materialFunction=/Game/Athena/Apollo/Environments/Landscape/Materials/Biome/CommonMaterials/MF_Common_Grass}
  n10 = Constant() {r=2}
  n11 = Saturate(None=Add(A=Multiply(A=@n9.GrassColorMask2, B=@n10), B=Constant() {r=-1.25}))
  n12 = Constant() {r=0.6}
  n13 = Saturate(None=Add(A=Multiply(A=@n11, B=@n12), B=@n9.GrassColorMask1))
  n14 = Multiply(A=@n13, B=Constant3Vector() {constant=<rgba 0 0 1 0>})
  n15 = Saturate(None=Add(A=@n13, B=Constant() {r=0.1}))
  n16 = CollectionParameter() {collection=/Game/Athena/Environments/Landscape/MPC/MPC_Landscape, parameterName=Ambient_WindMotion_Amount}
  n17 = ComponentMask(None=WorldPosition().XYZ) {g=true, r=true}
  n18 = Constant() {r=1200}
  n19 = MaterialFunctionCall() {materialFunction=/Game/Athena/Environments/Landscape/MaterialFunctions/MF_StormMask}
  n20 = LinearInterpolate(
   A=CollectionParameter() {collection=&MPC_Landscape, parameterName=Grass_WindSpeed}
   B=CollectionParameter() {collection=&MPC_Landscape, parameterName=Grass_WindSpeed_Storm}
   Alpha=@n19.Result
  )
  n21 = Time()
  n22 = TextureSample(UVs=Panner(Coordinate=Divide(A=@n17, B=@n18), Time=Multiply(A=@n20, B=@n21)) {speedY=1}) {samplerSource=SSM_Wrap_WorldGroupSettings, samplerType=SAMPLERTYPE_Normal, texture=/Game/Global/Textures/Default/Noise/T_SwirlNoise_N}
  n23 = LinearInterpolate(
   A=CollectionParameter() {collection=&MPC_Landscape, parameterName=Grass_WindPower}
   B=CollectionParameter() {collection=&MPC_Landscape, parameterName=Grass_WindPower_Storm}
   Alpha=@n19.Result
  )
  n24 = Multiply(A=Multiply(A=Multiply(A=@n15, B=@n16), B=Multiply(A=@n22.G, B=@n23)), B=Constant3Vector() {constant=<rgba 0 300 0 0>})
  n25 = StaticSwitchParameter(
   True=MaterialFunctionCall!(In) {materialFunction=/Game/Environments/Asteria/Foliage/Plants/BaseMaterial/MF_MediumPlantAnimation}.WPO Only
   False=Multiply(A=Saturate(None=@n8), B=Add(A=@n14, B=@n24))
  ) {desc="Enable to switch toward MF_MediumPlanntsAnimation ", group=ChooseMaterial, parameterName="Use New Wind "}
  n26 = Constant()
  n27 = Constant() {r=-50}
  n28 = Constant()
  n29 = PerInstanceFadeAmount()
  n30 = Add(
   A=ShadingPathSwitch(Default=@n25, Deferred=@n25, Forward=@n26, Mobile=@n26)
   B=Multiply(A=LinearInterpolate(A=@n27, B=@n28, Alpha=@n29), B=VertexNormalWS())
  )
  Reroute(None=@n30)
  n31 = MaterialFunctionCall(
   1=VectorParameter() {group=Per Instance Colors, parameterName=JP-DivColor1, defaultValue=<rgba 0.2104 0.5208 1 1>}.RGB
   2=VectorParameter() {group=Per Instance Colors, parameterName=JP-DivColor2, defaultValue=<rgba 0.45 0.739 0.75 1>}.RGB
   3=VectorParameter() {group=Per Instance Colors, parameterName=JP-DivColor3, defaultValue=<rgba 0.75 0.6863 0.225 1>}.RGB
   A=PerInstanceRandom()
  ) {materialFunction=&Lerp_Multiple_Float3}
  Reroute(None=@n31.Lerp 3 Inputs)
  ScalarParameter() {defaultValue=0.4, parameterName=WPO-MaskA}
  ScalarParameter() {defaultValue=-50, group=WPO-DistanceBased, parameterName=JunglePlants-SinkIntoGroundAmount}
  ScalarParameter() {parameterName=WPO-MaskB}
  n32 = TextureSample() {samplerSource=SSM_Wrap_WorldGroupSettings, texture=/Game/Environments/Apollo/Foliage/Nightmare/Textures/T_Nightmare_Grass_01_D}
  n33 = NamedRerouteDeclaration(None=@n31.Lerp 3 Inputs) {name=Random Colors, nodeColor=<rgba 0 0.6719 1 1>}
  n34 = TextureSample() {
   samplerSource=SSM_Wrap_WorldGroupSettings
   samplerType=SAMPLERTYPE_Masks
   texture=/Game/Environments/Apollo/Foliage/Nightmare/Textures/T_Nightmare_Grass_01_S
  }
  n35 = ScalarParameter() {defaultValue=1, parameterName=JunglePlants-SpecMod}
  n36 = ScalarParameter() {defaultValue=0.4, group=MF_Leaves, parameterName=Mobile-Leaves-Specular}
  n37 = Add(A=@n34.B, B=ScalarParameter() {defaultValue=0.2, parameterName=JunglePlants-RoughMod})
  n38 = ScalarParameter() {defaultValue=0.9, group=MF_Leaves, parameterName=Mobile-Leaves-Roughness}
  n39 = TextureSample() {samplerType=SAMPLERTYPE_Normal, texture=/Game/Environments/Apollo/Foliage/Nightmare/Textures/T_Nightmare_Grass_01_N}
  n40 = VertexNormalWS()
  SetMaterialAttributes!(
   MaterialAttributes
   BaseColor=Saturate(None=Divide(A=@n32.RGB, B=@n33))
   OpacityMask=@n32.A
   Specular=ShadingPathSwitch(Default=Multiply(A=@n34.R, B=@n35), Forward=@n36, Mobile=@n36)
   Roughness=ShadingPathSwitch(Default=Saturate(None=@n37), Forward=@n38, Mobile=@n38)
   Normal=ShadingPathSwitch(Default=Transform(None=@n39.RGB), Forward=@n40, Mobile=@n40)
   SubsurfaceColor=VectorParameter() {parameterName=JunglePlants-SSColor, defaultValue=<rgba 0.1305 0.1562 0.048 1>}.RGB
   Opacity=Constant() {r=1}
   WorldPositionOffset=NamedRerouteDeclaration(None=@n30) {name=WPO, nodeColor=<rgba 0 0.8359 1 1>}
  ) {attributeSetTypes [
    #1
    #6
    #3
    #4
    #2
    #7
    #8
    #5
   ]}
  StaticBoolParameter() {defaultValue=true, group=ChooseMaterial, parameterName=Mask is Diffuse A}
 ]}
