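# MF_LGRock is a switchable material function for small landscape rocks, outputting rock material attributes or passing its input through.
# Static switches choose UV textures or world-aligned Z projections; albedo is divided by a per-instance random color for variation.
# Roughness blends between near and far values, and world-space normals blend toward instance up, apparently by camera distance.
# A world-position offset sinks rocks as per-instance fade drops; several experimental branches remain unconnected.
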MaterialFunction /Game/Athena/Apollo/Environments/Landscape/SmallFoliage/MasterMaterial/MaterialFunctions/MF_LGRock
graph {nodes [
  Constant3Vector() {constant=<rgba 1.5 3.6358 6 1>}
  Constant3Vector() {constant=<rgba 1.2 3.7529 10 1>}
  Constant3Vector() {constant=<rgba 1.4 3.0246 7 1>}
  Constant() {r=0.25}
  Constant() {r=0.8}
  Constant() {r=-0.3}
  Constant() {r=0.5}
  Constant() {r=0.7}
  n1 = StaticSwitchParameter(
   True=NamedRerouteUsage()
   False=TextureSample(UVs=NamedRerouteUsage()) {samplerSource=SSM_Wrap_WorldGroupSettings, texture=/Game/Environments/Apollo/Rocks/Natural_Rocks/Textures/T_NaturalRock_AB_D}.RGB
  ) {group=Rock Textures, parameterName="Base Texture WS Z "}
  n2 = MaterialFunctionCall(
   1=VectorParameter() {group=LGRock, parameterName=RockDivColor1, defaultValue=<rgba 0.75 2.7672 5 1>}.RGB
   2=VectorParameter() {group=LGRock, parameterName=RockDivColor2, defaultValue=<rgba 0.9 3.3391 6 1>}.RGB
   3=VectorParameter() {group=LGRock, parameterName=RockDivColor3, defaultValue=<rgba 0.7337 2.3758 4 1>}.RGB
   A=PerInstanceRandom()
  ) {materialFunction=&Lerp_Multiple_Float3}
  n3 = ScalarParameter() {defaultValue=1, group=LGRock, parameterName=Rock-ScalarRoughness-Far}
  n4 = ScalarParameter() {defaultValue=0.7, group=LGRock, parameterName=Rock-ScalarRoughness-Near}
  n5 = NamedRerouteUsage()
  n6 = Constant3Vector() {constant=<rgba 0 0 1 0>}
  n7 = TextureSample(UVs=NamedRerouteUsage()) {
   samplerSource=SSM_Wrap_WorldGroupSettings
   samplerType=SAMPLERTYPE_Normal
   texture=/Game/Environments/Apollo/Rocks/Natural_Rocks/Textures/T_NaturalRock_AB_N
  }
  n8 = StaticSwitchParameter(
   True=NamedRerouteUsage()
   False=LinearInterpolate(A=Transform(None=@n7.RGB), B=VertexNormalWS(), Alpha=ScalarParameter() {group=LGRock, parameterName=Rock-BlendToVertexNormals})
  ) {group=Rock Textures, parameterName="Base Texture WS Z "}
  n9 = LinearInterpolate(
   A=ScalarParameter() {defaultValue=-0.1, group=LGRock, parameterName=Rock-BlendLerp-MIN}
   B=ScalarParameter() {defaultValue=0.8, group=LGRock, parameterName=Rock-BlendLerp-MAX}
   Alpha=NamedRerouteUsage()
  )
  n10 = SetMaterialAttributes!(
   MaterialAttributes
   BaseColor=Saturate(None=Divide(A=@n1, B=@n2.Lerp 3 Inputs))
   Specular=ScalarParameter() {defaultValue=0.5, group=LGRock, parameterName=Rock-ScalarSpec}
   Roughness=Saturate(None=LinearInterpolate(A=@n3, B=@n4, Alpha=@n5))
   Normal=LinearInterpolate(A=Transform(None=@n6) {desc=This looks goooooood, transformSourceType=TRANSFORMSOURCE_Instance}, B=@n8, Alpha=Saturate(None=@n9))
   WorldPositionOffset=NamedRerouteUsage()
  ) {attributeSetTypes [
    #1
    #2
    #3
    #4
    #5
   ]}
  n11 = SetMaterialAttributes() {attributeSetTypes [
    #1
    #2
    #3
    #4
    #5
   ]}
  FunctionOutput(None=StaticSwitchParameter(True=@n10, False=FunctionInput(Preview=@n11) {bUsePreviewValueAsDefault=true, inputType=FunctionInput_MaterialAttributes}) {group=ChooseMaterial, parameterName=Use-MF_LGRock}) {desc="Normals in the parent material are in worldspace!"}
  n12 = TextureSample(UVs=NamedRerouteUsage()) {
   samplerSource=SSM_Wrap_WorldGroupSettings
   samplerType=SAMPLERTYPE_Masks
   texture=/Game/Environments/Apollo/Rocks/Natural_Rocks/Textures/T_NaturalRock_AB_S
  }
  MaterialFunctionCall(Float3=StaticSwitchParameter(True=NamedRerouteUsage(), False=@n12.RGB) {group=Rock Textures, parameterName="Base Texture WS Z "}) {materialFunction=&BreakOutFloat3Components}
  MaterialFunctionCall(
   1=Constant3Vector() {constant=<rgba 0.2243 0 1 1>}
   2=Constant3Vector() {constant=<rgba 0 1 0 0>}
   3=Constant3Vector() {constant=<rgba 0 0 1 0>}
   4=Constant3Vector() {constant=<rgba 1 0.5 0 1>}
   A=PerInstanceRandom()
  ) {materialFunction=&Lerp_Multiple_Float3}
  n13 = TextureObjectParameter() {group=Rock Textures, parameterName=Rock-Diffuse, texture=&T_NaturalRock_AB_D}
  n14 = TextureObjectParameter() {
   group=Rock Textures
   parameterName=Rock-SpecularMasks
   samplerType=SAMPLERTYPE_Masks
   texture=&T_NaturalRock_AB_S
  }
  n15 = ScalarParameter() {defaultValue=1, group=LGRock, parameterName=Rock-BaseTextureScale}
  n16 = TextureObjectParameter() {
   group=Rock Textures
   parameterName=Rock-Normals
   samplerType=SAMPLERTYPE_Normal
   texture=&T_NaturalRock_AB_N
  }
  n17 = MaterialFunctionCall!(
   Diffuse Texture Object=NamedRerouteDeclaration(None=@n13) {name=Object-Diffuse, nodeColor=<rgba 1 0 1 1>}
   Spec Texture Object=NamedRerouteDeclaration(None=@n14) {name=Object-SpecMasks, nodeColor=<rgba 1 0 1 1>}
   TextureSize=@n15
   --------------------------
   Use seperate settings for Z projection
   Use seperate texture for Z projection
   Texture Size Z projection
   Z Plane Diffuse Texture Object
   Z Plane Spec Texture Object
   "--------------------------      "
   ProjectionTransitionContrast
   Use Texture Alpha Heightmap to Refine the Transistion
   HeightLerpTransitionContrast
   PickTheDiffuseHeightChannel
   TheDiffuseHeightChannel
   "--------------------------       "
   NormalTextureObject=NamedRerouteDeclaration(None=@n16) {name=Object-Normal, nodeColor=<rgba 1 0 1 1>}
   World Space Normals=StaticBool() {value=true}
   "--------------------------  "
   Export Float 4 Diffuse
   Export Float 4 Spec
   "--------------------------     "
   WorldPosition
   World Space Normal
  ) {materialFunction=/Engine/Functions/Engine_MaterialFunctions02/Texturing/WorldAlignedTextures_Complex}
  NamedRerouteDeclaration(None=@n17.Z Spec) {name=Z-Spec, nodeColor=<rgba 1 0 1 1>}
  NamedRerouteDeclaration(None=@n17.Normal Z Texture) {name=Z-Normal, nodeColor=<rgba 1 0 1 1>}
  NamedRerouteDeclaration(None=@n17.Z Diffuse Texture) {name=Z-Diffuse, nodeColor=<rgba 1 0 1 1>}
  NamedRerouteDeclaration(None=Multiply(A=TextureCoordinate(), B=@n15)) {name=BaseTextureTile, nodeColor=<rgba 1 0 1 1>}
  n18 = ScalarParameter() {defaultValue=-20, group=WPO-DistanceBased, parameterName=DistanceGroundSink}
  n19 = Constant()
  n20 = PerInstanceFadeAmount()
  NamedRerouteDeclaration(None=Multiply(A=LinearInterpolate(A=@n18, B=@n19, Alpha=@n20), B=Constant3Vector() {constant=<rgba 0 0 1 0>})) {name=WPO, nodeColor=<rgba 1 0 1 1>}
  NamedRerouteDeclaration(None=SphereMask(A=CameraPositionWS(), B=WorldPosition().XYZ, Radius=Constant() {r=10000}, Hardness=Constant())) {name=CameraDistanceFade, nodeColor=<rgba 1 0 1 1>}
  NamedRerouteUsage()
  NamedRerouteUsage()
  NamedRerouteUsage()
  Normalize(VectorInput=LinearInterpolate(
    A=@n8
    B=Constant3Vector() {constant=<rgba 0 0 1 0>}
    Alpha=ScalarParameter() {defaultValue=0.25, group=LGRock, parameterName=Rock-BlendFlatWorldNormal}
   ))
  n21 = GetMaterialAttributes() {attributeGetTypes [#1, #4, #2]}
  n22 = SphereMask(A=CameraPositionWS(), B=WorldPosition().XYZ, Radius=Constant() {r=6500}, Hardness=Constant())
  n23 = LinearInterpolate(
   A=Transform(None=Constant3Vector() {constant=<rgba 0 0 1 0>}) {transformSourceType=TRANSFORMSOURCE_World, transformType=TRANSFORM_Tangent}
   B=@n21.Normal
   Alpha=@n22
  )
  n24 = Constant() {r=0.1}
  SetMaterialAttributes!(
   MaterialAttributes=@n21.MaterialAttributes
   BaseColor=LinearInterpolate(A=Constant3Vector() {desc=Distance fade color, constant=<rgba 0.0765 0.1198 0.0236 1>}, B=@n21.BaseColor, Alpha=@n22)
   Normal=QualitySwitch(Default=@n23, Low=@n21.Normal, High=@n23, Medium=@n23)
   Specular=Multiply(A=Max(A=@n24, B=@n22), B=@n21.Specular)
   WorldPositionOffset
  ) {attributeSetTypes [
    #1
    #4
    #2
    #5
   ]}
 ]}
