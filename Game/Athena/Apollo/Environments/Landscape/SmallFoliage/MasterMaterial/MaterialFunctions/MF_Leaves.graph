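# Leaves material function for Apollo small foliage, toggled by a static switch that otherwise passes input material attributes through.
# A UV1-sampled, offset LUT tints leaf albedo and drives subsurface color, while albedo alpha supplies the opacity mask.
# World position offset pushes vertices near the camera outward and down, and offsets them along normals by per-instance fade.
# Forward and mobile shading substitute vertex normals for the normal map.
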MaterialFunction /Game/Athena/Apollo/Environments/Landscape/SmallFoliage/MasterMaterial/MaterialFunctions/MF_Leaves
graph {nodes [
  Constant() {r=0.2}
  Constant() {r=0.9}
  n1 = Constant()
  n2 = ScalarParameter() {defaultValue=0.25, parameterName=LUT offset}
  n3 = TextureSampleParameter2D(UVs=Add(A=TextureCoordinate() {coordinateIndex=1}, B=AppendVector(A=@n1, B=@n2))) {parameterName=LUT, texture=/Game/Athena/Apollo/Environments/Landscape/SmallFoliage/Assets/Leaves/T_Apollo_Leaves_LUT}
  n4 = TextureSampleParameter2D() {parameterName=BC_Leaves, texture=/Game/Athena/Apollo/Environments/Landscape/SmallFoliage/Assets/Leaves/T_Apollo_Leaves_D}
  n5 = TextureSampleParameter2D() {
   parameterName=N_Leaves
   samplerType=SAMPLERTYPE_Normal
   texture=/Game/Athena/Apollo/Environments/Landscape/SmallFoliage/Assets/Leaves/T_Apollo_Leaves_N
  }
  n6 = VertexNormalWS()
  n7 = SphereMask(A=CameraPositionWS(), B=WorldPosition().XYZ, Radius=Constant() {r=200}, Hardness=Constant())
  n8 = WorldPosition()
  n9 = CameraPositionWS()
  n10 = Add(A=Multiply(A=Subtract(A=@n8.XYZ, B=@n9), B=Constant3Vector() {constant=<rgba 1 1 0 0>}), B=Constant3Vector() {constant=<rgba 0 0 -200 0>})
  n11 = LinearInterpolate(A=Constant() {r=-50}, B=Constant(), Alpha=PerInstanceFadeAmount())
  n12 = VertexNormalWS()
  n13 = SetMaterialAttributes!(
   MaterialAttributes
   BaseColor=Multiply(A=@n3.RGB, B=@n4.RGB)
   OpacityMask=@n4.A
   Specular=ScalarParameter() {defaultValue=0.4, group=MF_Leaves, parameterName=Leaves-Specular}
   Roughness=ScalarParameter() {defaultValue=0.9, group=MF_Leaves, parameterName=Leaves-Roughness}
   Normal=ShadingPathSwitch(Default=Transform(None=@n5.RGB), Forward=@n6, Mobile=@n6)
   SubsurfaceColor=Multiply(A=@n3.RGB, B=Constant() {r=0.75})
   Opacity=Constant() {r=1}
   WorldPositionOffset=Add(A=Multiply(A=@n7, B=@n10), B=Multiply(A=@n11, B=@n12))
  ) {attributeSetTypes [
    #1
    #2
    #3
    #4
    #5
    #6
    #7
    #8
   ]}
  FunctionOutput(None=StaticSwitchParameter(True=@n13, False=FunctionInput() {inputType=FunctionInput_MaterialAttributes}) {group=ChooseMaterial, parameterName=Use-MF_Leaves})
 ]}
