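# Swamp grass layer for landscape small foliage that sets masked foliage attributes, or passes its input through when Use-MF_SwampGrass is off.
# Color blends along UV1 with optional per-instance variation and diffuse tint; opacity comes from diffuse alpha or a channel mask.
# Normals optionally force Z-up.
# Common grass landscape masks drive wind falloff and vertical offset in the gameplay wind WPO, which then adds global physics field displacement.
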MaterialFunction /Game/Athena/Apollo/Environments/Landscape/SmallFoliage/MasterMaterial/MaterialFunctions/MF_SwampGrass
graph {nodes [
  n1 = TextureSampleParameter2D() {group=SwampGrass, parameterName=Diffuse, samplerSource=SSM_Wrap_WorldGroupSettings, texture=/Engine/EngineMaterials/DefaultDiffuse}
  n2 = ComponentMask(None=TextureCoordinate() {coordinateIndex=1}) {g=true}
  n3 = LinearInterpolate(
   A=VectorParameter() {group=SwampGrass, parameterName=SwampGrass-Color, defaultValue=<rgba 0.1 0.0874 0.005 1>}.RGB
   B=VectorParameter() {group=SwampGrass, parameterName=SwampGrass-Color2, defaultValue=<rgba 0.0554 0.08 0.01 1>}.RGB
   Alpha=@n2
  )
  n4 = VectorParameter() {group=SwampGrass, parameterName=SwampGrass-B-Color, defaultValue=<rgba 0.1 0.0874 0.005 1>}
  n5 = VectorParameter() {group=SwampGrass, parameterName=SwampGrass-B-Color2, defaultValue=<rgba 0.0554 0.08 0.01 1>}
  n6 = StaticSwitchParameter(True=LinearInterpolate(A=@n3, B=LinearInterpolate(A=@n4.RGB, B=@n5.RGB, Alpha=@n2), Alpha=PerInstanceRandom()), False=@n3) {group=SwampGrass, parameterName=SwampGrass-UsePIR-Color}
  n7 = TextureSampleParameter2D() {
   group=SwampGrass
   parameterName=SwampGrass--OpacityMask
   samplerSource=SSM_Wrap_WorldGroupSettings
   samplerType=SAMPLERTYPE_Masks
   texture=/Game/Athena/Apollo/Environments/Landscape/SmallFoliage/Assets/SwampGrass/Swamp_Reeds_M
  }
  n8 = ChannelMaskParameter(None=@n7.RGBA) {group=SwampGrass, parameterName=ChannelMask}
  n9 = Transform(None=Constant3Vector() {constant=<rgba 0 0 1 0>}) {transformSourceType=TRANSFORMSOURCE_Local}
  n10 = TextureSampleParameter2D() {
   group=SwampGrass
   parameterName=SwampGrass-Normal
   samplerSource=SSM_Wrap_WorldGroupSettings
   samplerType=SAMPLERTYPE_Normal
   texture=/Game/Athena/Apollo/Environments/Landscape/SmallFoliage/Assets/SwampGrass/Swamp_Reeds_N
  }
  n11 = ScalarParameter() {defaultValue=1, group=SwampGrass, parameterName=SwampGrass-NormalIntensity}
  n12 = MaterialFunctionCall(Normal=@n10.RGB, Flatness=@n11) {materialFunction=&FlattenNormal}
  n13 = Multiply(A=@n10.RGB, B=VectorParameter() {group=SwampGrass, parameterName=MF_SwampGrass_NormalTextureAmt, defaultValue=<rgba 0.15 0.15 1 1>}.RGB)
  n14 = MaterialFunctionCall(BaseNormal=Constant3Vector() {constant=<rgba 0 0 1 0>}, AdditionalNormal=@n13) {materialFunction=&BlendAngleCorrectedNormals}
  n15 = Constant() {r=0.1}
  n16 = MaterialFunctionCall!(CombinedLandscapeLayersMask, BypassUseRVT, BypassBiomeColors) {materialFunction=/Game/Athena/Apollo/Environments/Landscape/Materials/Biome/CommonMaterials/MF_Common_Grass}
  n17 = Constant() {r=2}
  n18 = Saturate(None=Add(A=Multiply(A=@n16.GrassColorMask2, B=@n17), B=Constant() {r=-1.25}))
  n19 = Constant() {r=0.6}
  n20 = Saturate(None=Add(A=Multiply(A=@n18, B=@n19), B=@n16.GrassColorMask1))
  n21 = LinearInterpolate(
   A=ScalarParameter() {defaultValue=1, group=SwampGrass, parameterName=MF_SwampGrass-wpoa}
   B=ScalarParameter() {defaultValue=-1, group=SwampGrass, parameterName=MF_SwampGrass-wpob}
   Alpha=ComponentMask(None=TextureCoordinate() {coordinateIndex=1}) {g=true}
  ) {constA=1.2, constB=-1}
  n22 = Multiply(A=@n20, B=Constant3Vector() {constant=<rgba 0 0 1 0>})
  n23 = MaterialFunctionCall!(
   SinkIntoGroundDistance
   WindFalloffMask=Saturate(None=Add(A=@n20, B=@n15))
   AfterHeightModification=Multiply(A=Saturate(None=@n21), B=Add(A=@n22, B=@n23.Wind))
  ) {materialFunction=&/MF_GrassWind-GameplayWPO}
  n24 = MaterialFunctionCall!(
   Previous WPO Value=@n23.FinalWPO
   Force Magnitude Mask
   Force Magnitude Multiplier
   Project Field Force Vector to Normals
   Noise Size
   Noise Pan Speed
   Noise Falloff
   Stretch Noise Coordinates Z
  ) {desc=Sample global physics field add to WPO, materialFunction=/Game/Effects/Niagara/MasterMaterials/Functions/MF_FX_PhysicsFields_Grass}
  n25 = SetMaterialAttributes!(
   MaterialAttributes
   BaseColor=StaticSwitchParameter(True=Multiply(A=@n1.RGB, B=@n6), False=@n6) {group=SwampGrass, parameterName=UseDiffuse}
   Specular=ScalarParameter() {defaultValue=0.05, group=SwampGrass, parameterName=SwampGrass-Specular}
   Roughness=ScalarParameter() {defaultValue=0.8, group=SwampGrass, parameterName=SwampGrass-Roughness}
   OpacityMask=Multiply(
    A=StaticSwitchParameter(True=@n1.A, False=@n8.R) {group=SwampGrass, parameterName=UseDiffuse}
    B=ScalarParameter() {defaultValue=1.1, group=SwampGrass, parameterName=SwampGrass-OpacityMaskOffset}
   )
   Normal=StaticSwitchParameter(
    True=MaterialFunctionCall(BaseNormal=@n9, AdditionalNormal=@n12.Result) {materialFunction=&BlendAngleCorrectedNormals}.Result
    False=Transform(None=@n14.Result)
   ) {group=SwampGrass, parameterName=UseForcedZup}
   WorldPositionOffset=@n24.Result
   Opacity=Constant() {r=0.1}
  ) {attributeSetTypes [
    #1
    #2
    #3
    #4
    #5
    #6
    #7
   ]}
  FunctionOutput(None=StaticSwitchParameter(True=@n25, False=FunctionInput() {inputType=FunctionInput_MaterialAttributes}) {group=ChooseMaterial, parameterName=Use-MF_SwampGrass})
  n26 = VertexNormalWS()
  LinearInterpolate(
   A=@n26
   B=MaterialFunctionCall(BaseNormal=@n26, AdditionalNormal=Transform(None=@n13)) {materialFunction=&BlendAngleCorrectedNormals}.Result
   Alpha=@n11
  )
 ]}
